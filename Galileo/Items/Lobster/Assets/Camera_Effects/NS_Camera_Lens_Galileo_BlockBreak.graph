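# Niagara system for a camera lens effect, by name shown on block break for the Galileo Lobster item.
# A single emitter spawns an instantaneous burst and initializes bright, non-uniformly sized particles with a material random value, feeding both a sprite and a mesh renderer.
# Particles loop their lifetime instead of dying, and no motion modules or user parameters are present.
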NiagaraSystem /Galileo/Items/Lobster/Assets/Camera_Effects/NS_Camera_Lens_Galileo_BlockBreak
dependencies {usedDataInterfaces [], usedDynamicInputs [], usedModules [&SystemState]}
emitters {Empty {
  emitterUpdateScript [
   &EmitterState()
   &SpawnBurst_Instantaneous(Use Spawn Probability=false, UseLoopCountLimit=true)
  ]
  particleSpawnScript [&InitializeParticle(
    Color=<rgba 3 3 3 1>
    Mesh Scale=&Multiply_VectorByFloat(Vector=<10 18 1>, Float=1.8)
    Mesh Scale Mode=::NewEnumerator4
    Sprite Size=<180 100>
    Sprite Size Mode=::NewEnumerator4
    UsePositionOffset=false
    Write Material Random=true
   )]
  particleUpdateScript [&ParticleState(Kill Particles When Lifetime Has Elapsed=false, Loop Particles Lifetime=true)]
  rendererClasses [&NiagaraSpriteRendererProperties, &NiagaraMeshRendererProperties]
  renderers [
   {rendererClass=&NiagaraSpriteRendererProperties}
   {rendererClass=&NiagaraMeshRendererProperties, rendererIndex=1}
  ]
 }}
summary {
 systemName=NS_Camera_Lens_Galileo_BlockBreak
 emitters [{emitterName=Empty, rendererClasses [&NiagaraSpriteRendererProperties, &NiagaraMeshRendererProperties]}]
 userVariables []
}
